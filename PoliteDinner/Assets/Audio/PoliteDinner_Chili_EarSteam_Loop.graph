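# Stereo MetaSound source for the PoliteDinner chili ear-steam consumable sound, layering an alarm one-shot with a steam loop.
# On play, each layer picks a wave from its own array, then rolls slight random pitch and gain before triggering its wave player.
# A two-channel stereo mixer applies the gains and sums both layers to the output.
# Only the steam player's finish fires OnFinished, under consumable attenuation and sound class.
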MetaSoundSource /PoliteDinner/Assets/Audio/PoliteDinner_Chili_EarSteam_Loop
inputs [
 Alarm OS {type=WaveAsset:Array}
 Steam Loop {type=WaveAsset:Array}
 UE.Source.OnPlay {type=Trigger}
]
interfaces [UE.OutputFormat.Stereo:1.0, UE.Source.OneShot:1.0, UE.Source:1.1]
outputs [
 "UE.OutputFormat.Stereo.Audio:0" {type=Audio}
 "UE.OutputFormat.Stereo.Audio:1" {type=Audio}
 UE.Source.OneShot.OnFinished {type=Trigger}
]
pages [Default {
  input_defaults {
   UE.Source.OnPlay=null
   Alarm OS [&/PoliteDinner_Chili_EarSteam_Alarm_01]
   Steam Loop [&/PoliteDinner_Chili_EarSteam_Loop_01]
  }
  nodes [
   ArrayRandomGet_0 = Array.Random Get:WaveAsset:Array(In Array=^Alarm OS, Next=^UE.Source.OnPlay)
   n1 = UE.RandomFloat(Next=@ArrayRandomGet_0.On Next) {also {defaults {Max=0.5, Min=-0.5}}}
   n2 = UE.RandomFloat(Next=UE.Reroute(Value=@n1.On Next).Value) {also {defaults {Max=1.05, Min=0.95}}}
   ArrayRandomGet_1 = Array.Random Get:WaveAsset:Array(In Array=^Steam Loop, Next=^UE.Source.OnPlay) {also {defaults {Weights []}}}
   n3 = UE.RandomFloat(Next=@ArrayRandomGet_1.On Next) {also {defaults {Max=0.5, Min=-0.5}}}
   n4 = UE.RandomFloat(Next=UE.Reroute(Value=@n3.On Next).Value) {also {defaults {Max=1.05, Min=0.95}}}
   WavePlayer_0 = UE.Wave Player:Stereo(Wave Asset=@ArrayRandomGet_1.Value, Pitch Shift=@n3.Value, Play=@n4.On Next) {also {defaults {Loop="<MetaSoundBuilderResult.SUCCEEDED: 0>"}, version [1,1]}}
   WavePlayer_1 = UE.Wave Player:Stereo(Wave Asset=@ArrayRandomGet_0.Value, Play=@n2.On Next, Pitch Shift=@n1.Value) {also {version [1,1]}}
   AudioMixerAudioMixerStereo2_0 = "AudioMixer.Audio Mixer (Stereo, 2)"(
    Gain 0=@n2.Value
    Gain 1=@n4.Value
    In 1 L=@WavePlayer_0.Out Left
    In 1 R=@WavePlayer_0.Out Right
    In 0 L=@WavePlayer_1.Out Left
    In 0 R=@WavePlayer_1.Out Right
   )
  ]
  links [
   {dst="@UE.OutputFormat.Stereo.Audio:0", src=AudioMixerAudioMixerStereo2_0.Out L}
   {dst="@UE.OutputFormat.Stereo.Audio:1", src=AudioMixerAudioMixerStereo2_0.Out R}
   {dst="@UE.Source.OneShot.OnFinished", src=WavePlayer_0.On Finished}
  ]
 }]
properties {
 assetImportData=&self:AssetImportData
 attenuationSettings=&Fort_Consumable_Use_Athena_Att
 maxDistance=2250
 numChannels=2
 outputFormat=Stereo
 sampleRate=48000
 soundClassObject=&Fort_Consumables
}
