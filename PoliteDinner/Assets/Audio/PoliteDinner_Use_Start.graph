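# One-shot stereo MetaSound source for the start of using PoliteDinner, routed through the consumables sound class and consumable-use attenuation.
# On play, it picks a random wave from the Sound array, which holds a single sample by default.
# A random float then applies a slight pitch shift up or down and triggers the wave player.
# The player feeds both output channels and signals the one-shot finish when done.
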MetaSoundSource /PoliteDinner/Assets/Audio/PoliteDinner_Use_Start
inputs [
 Sound {type=WaveAsset:Array}
 UE.Source.OnPlay {type=Trigger}
]
interfaces [UE.OutputFormat.Stereo:1.0, UE.Source.OneShot:1.0, UE.Source:1.1]
outputs [
 "UE.OutputFormat.Stereo.Audio:0" {type=Audio}
 "UE.OutputFormat.Stereo.Audio:1" {type=Audio}
 UE.Source.OneShot.OnFinished {type=Trigger}
]
pages [Default {
  input_defaults {UE.Source.OnPlay=null, Sound [&/PoliteDinner_Use_Start_01]}
  nodes [
   ArrayRandomGet_0 = Array.Random Get:WaveAsset:Array(In Array=^Sound, Next=^UE.Source.OnPlay)
   n1 = UE.RandomFloat(Next=@ArrayRandomGet_0.On Next) {also {defaults {Max=0.5, Min=-0.5}}}
   WavePlayer_0 = UE.Wave Player:Stereo(Wave Asset=@ArrayRandomGet_0.Value, Play=@n1.On Next, Pitch Shift=@n1.Value) {also {version [1,1]}}
  ]
  links [
   {dst="@UE.Source.OneShot.OnFinished", src=WavePlayer_0.On Finished}
   {dst="@UE.OutputFormat.Stereo.Audio:0", src=WavePlayer_0.Out Left}
   {dst="@UE.OutputFormat.Stereo.Audio:1", src=WavePlayer_0.Out Right}
  ]
 }]
properties {
 assetImportData=&self:AssetImportData
 attenuationSettings=&Fort_Consumable_Use_Athena_Att
 maxDistance=2250
 numChannels=2
 outputFormat=Stereo
 soundClassObject=&Fort_Consumables
}
