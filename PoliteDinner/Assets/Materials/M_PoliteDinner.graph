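# Masked character master material for PoliteDinner that adds a staged vertex deformation between shared CharacterParent PreFX and PostFX functions.
# The Stage parameter drives blended progress, quantized into 24 steps for a stop-motion-like scale and vertical shift around a raised object pivot.
# A red-dominant vertex-color mask limits this World Position Offset.
# A TimeNative reroute normalizes time since GameTimeFromBP over AnimationDuration; the Progress parameter is unconnected.
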Material /PoliteDinner/Assets/Materials/M_PoliteDinner
graph {
 properties {
  bUseMaterialAttributes=true
  bUsedWithClothing=true
  bUsedWithMeshDeformer=true
  bUsedWithMorphTargets=true
  bUsedWithNanite=true
  bUsedWithNiagaraMeshParticles=true
  bUsedWithSkeletalMesh=true
  blendMode=BLEND_Masked
  pixelDepthOffsetMode=PDOM_Legacy
  shadingModel=MSM_FromMaterialExpression
  usedShadingModels=Default Lit | Subsurface | Clear Coat | Subsurface Profile | Cloth
 }
 nodes [
  n1 = MaterialFunctionCall() {materialFunction=&MF_CharacterParent_2023_PreFX}
  n2 = Saturate(None=DotProduct(A=VertexColor(), B=Constant3Vector() {constant=<rgba 1 -1 -1 0>}))
  n3 = ScalarParameter() {group=Weakpoints, parameterName=Stage}
  n4 = Subtract(A=@n3)
  n5 = NamedRerouteUsage()
  n6 = NamedRerouteDeclaration(None=Saturate(None=LinearInterpolate(A=@n4, B=@n3, Alpha=@n5) {constB=20})) {name=Stages, nodeColor=<rgba 0.3812 1 0.0069 1>}
  n7 = Constant() {r=24}
  n8 = Multiply(A=@n6, B=@n7) {constB=10}
  n9 = OneMinus(None=Divide(A=Floor(None=@n8), B=@n7) {constB=10})
  n10 = WorldPosition() {worldPositionShaderOffset=WPT_ExcludeAllShaderOffsets}
  n11 = Add(
   A=MaterialFunctionCall() {materialFunction=&ObjectPivotPoint}.Object Pivot Location
   B=Constant3Vector() {constant=<rgba 0 0 20 0>}
  )
  n12 = Subtract(A=LinearInterpolate(A=LinearInterpolate(Alpha=@n9) {constA=0.75}, Alpha=Step(X=@n6) {constX=0.8, constY=0.65}) {constA=0.75, constB=0})
  n13 = Add(
   A=MaterialFunctionCall(Z=LinearInterpolate(Alpha=@n9) {constA=-37, constB=0}) {materialFunction=&MakeFloat3}.Result
   B=Multiply(A=Subtract(A=@n10.XYZ, B=@n11), B=Multiply(A=@n12) {constB=1.9})
  )
  MaterialFunctionCall(
   Material=SetMaterialAttributes!(MaterialAttributes=@n1.Result, World Position Offset=Multiply(A=@n2, B=@n13), Opacity Mask) {attributeSetTypes [#1, #2]}
   M=@n1.M
   Pre Skinned Local Position=@n1.M
   Distance Blend=@n1.M
   Fresnel=@n1.M
  ) {materialFunction=&MF_CharacterParent_2023_PostFX}
  n14 = Subtract(A=Time(), B=ScalarParameter() {defaultValue=-99999, group=Weakpoints, parameterName=GameTimeFromBP})
  n15 = Add(A=ScalarParameter() {defaultValue=1, group=Weakpoints, parameterName=AnimationDuration, sortPriority=101}) {constB=0}
  NamedRerouteDeclaration(None=Saturate(None=Divide(A=@n14, B=@n15))) {name=TimeNative, nodeColor=<rgba 0.2891 1 0 1>}
  ScalarParameter() {group=PoliteDinner, parameterName=Progress}
 ]
}
