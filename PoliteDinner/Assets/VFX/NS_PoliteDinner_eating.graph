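# Niagara system NS_PoliteDinner_eating, by its name the eating effect for PoliteDinner, with emitters named after ramen-like food debris.
# Six emitters make up the burst: two noodle-bit emitters, two vegetable emitters, a chasyu (pork) emitter and a splash ("Sprash").
# The system exposes no user parameters, and SystemState is the only module recorded.
# Renderer lists are empty here, so the look and motion of each emitter are not shown.
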NiagaraSystem /PoliteDinner/Assets/VFX/NS_PoliteDinner_eating
dependencies {usedDataInterfaces [], usedDynamicInputs [], usedModules [&SystemState]}
emitters {
 NoodleBits {rendererClasses [], renderers []}
 NoodleBits001 {rendererClasses [], renderers []}
 Sprash {rendererClasses [], renderers []}
 chasyu {rendererClasses [], renderers []}
 vege1 {rendererClasses [], renderers []}
 vege2 {rendererClasses [], renderers []}
}
summary {
 systemName=NS_PoliteDinner_eating
 emitters [
  {emitterName=vege2, rendererClasses []}
  {emitterName=vege1, rendererClasses []}
  {emitterName=chasyu, rendererClasses []}
  {emitterName=NoodleBits, rendererClasses []}
  {emitterName=NoodleBits001, rendererClasses []}
  {emitterName=Sprash, rendererClasses []}
 ]
 userVariables []
}
