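# Prototype steam effect for PoliteDinner, drawn as a rising, widening Niagara ribbon.
# Smoke_Ribbin001 sways spawn velocity with a sine on emitter age, adds panning curl noise and makes particles follow source movement.
# Alpha combines a particle fade curve with an emitter-age ramp, dimmed as owner speed increases; distant particles are culled.
# The sprite emitter Smoke_top and the second spawn-group rate are disabled, and the Steam emitter has no renderers.
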NiagaraSystem /PoliteDinner/Assets/VFX/NS_Proto_PoliteDinner_Steam
dependencies {
 usedDataInterfaces [&NiagaraDataInterfaceCurve, &NiagaraDataInterfaceArrayFloat]
 usedDynamicInputs []
 usedModules [&SystemState]
}
emitters {
 Smoke_Ribbin001 {
  emitterUpdateScript [
   &EmitterState(
    Enable Distance Culling=true
    Life Cycle Mode=::NewEnumerator1
    Loop Duration=1
    Max Distance Response=::SleepAndClearParticles
    MaxDistance=2000
    Module.bUseMaxDistance=true
    Module.bUseMinDistance=false
    Scalability Mode=::NewEnumerator1
    UseLoopDelay=false
   )
   &SpawnRate(SpawnRate=24, Use Spawn Probability=false)
   &SpawnRate(SpawnGroup=1, SpawnRate=17, Use Spawn Probability=false) {enabled=false, module=SpawnRate001}
  ]
  particleSpawnScript [
   &InitializeParticle(
    Color=&MultiplyLinearColor(
     Linear Color=&RandomRangeLinearColor(Minimum=<rgba 0.1236 0.1198 0.125 1>, Maximum=<rgba 0.2348 0.2396 0.2396 1>)
     Scale Factor=<1 1 1 1>
    )
    Lifetime=&MaskFloatBySpawnGroup(Masked Float=1.18, Passthrough Float=0.9)
    Position Offset=<0 0 -3>
    Position Offset Coordinate Space=::Local
    Ribbon Width=66
    Ribbon Width Mode=::NewEnumerator1
    Sprite Rotation Mode=::NewEnumerator2
    UsePositionOffset=true
    Write Material Random=true
   )
   /SetVariables(Particles.RibbonID=/Niagara/DynamicInputs/Meshes/MakeRibbonIDFromSpawnGroup()) {isSetParameters=true}
   &AddVelocity(
    Rotation Coordinate Space=::World
    Velocity=&MakeVector(
     X=&Sine(Normalized Angle=$Emitter.Age, Period=1, Scale=2, Bias=-1)
     Y=0
     Z=&SelectFloatFromArray(Direct Array Index=$Emitter.SpawnGroup, Float Selection Array=DataInterface() {propertyValues {FloatData [60,40]}})
    )
   )
   &AddVelocity(Velocity=<0 0 4>) {module=AddVelocity001}
  ]
  particleUpdateScript [
   &ParticleState()
   &CurlNoiseForce(
    Noise Frequency=200
    Noise Quality / Cost=::NewEnumerator0
    Noise Strength=100
    Pan Noise=true
    Pan Noise Field=<-1 1 5>
    Randomization Vector:Vector3f=<0.64 0.45 0.2>
   )
   &InheritSourceMovement(Applied Position Scale=<0.8 0.8 0.9>, ApplyForce=false, ApplyPosition=true, ApplyPositionFalloff=false)
   &SolveForcesAndVelocity(Clamp Velocity=false, Limit Acceleration=false)
   &ScaleColor(
    Scale Alpha=&Multiply_Float(
     A=&FloatFromCurve(
      FloatCurve=DataInterface() {propertyValues {Curve {keys [
          {arriveTangent=1.6049, interpMode=RCIM_Cubic, leaveTangent=1.6049, tangentMode=RCTM_User, time=0.0038, value=0}
          {arriveTangent=0.9841, interpMode=RCIM_Cubic, leaveTangent=0.9841, time=0.1836, value=0.8472}
          {arriveTangent=-1.3646, interpMode=RCIM_Cubic, leaveTangent=-1.3646, tangentMode=RCTM_User, time=0.334, value=0.325}
          {tangentMode=RCTM_User, time=1, value=0}
         ]}}}
      Scale Curve=0.1
     )
     B=&FloatFromCurve(
      FloatCurve=DataInterface() {propertyValues {Curve {keys [<at 0.2 0 RCIM_Cubic>, {arriveTangent=3.0097, tangentMode=RCTM_User, time=0.8, value=1}]}}}
      CurveIndex=$Emitter.Age
      Scale Curve=&Multiply_Float(
       A=&RemapRange(
        Input Value To Remap=&VectorLength(Vector=$Engine.Owner.Velocity)
        Input Min=50
        Input Max=0
        Output Min=0.7
        Output Max=1
       )
       B=&RandomRangeFloat(Minimum=1, Maximum=1)
      )
     )
    )
    Scale RGB=&VectorFromFloat(Value=&FloatFromCurve(
      FloatCurve=DataInterface() {propertyValues {Curve {keys [
          <at -0.0156 1.5373 RCIM_Cubic>
          {arriveTangent=0.1701, interpMode=RCIM_Cubic, leaveTangent=0.1701, time=0.1272, value=1.6298}
          {arriveTangent=-0.2403, interpMode=RCIM_Cubic, leaveTangent=-0.2403, time=0.5, value=1.625}
          <at 1 1.4201 RCIM_Cubic>
         ]}}}
      Scale Curve=2
     ))
    ScaleA=true
    ScaleRGB=true
   )
   &ScaleRibbonWidth(Ribbon Width Scale=&FloatFromCurve(FloatCurve=DataInterface() {propertyValues {Curve {keys [
         {arriveTangent=0.0536, interpMode=RCIM_Cubic, leaveTangent=0.0536, tangentMode=RCTM_User, time=0, value=0.1}
         {arriveTangent=1.7358, interpMode=RCIM_Cubic, leaveTangent=1.7358, tangentMode=RCTM_User, time=0.2, value=0.25}
         {arriveTangent=0.9375, interpMode=RCIM_Cubic, leaveTangent=0.9375, time=0.3694, value=0.25}
         {arriveTangent=2.7752, interpMode=RCIM_Cubic, leaveTangent=2.7752, time=0.6, value=0.625}
         {arriveTangent=6.2786, interpMode=RCIM_Cubic, leaveTangent=6.2786, tangentMode=RCTM_User, time=1, value=2}
        ]}}}))
  ]
  rendererClasses [&NiagaraRibbonRendererProperties]
  renderers [{rendererClass=&NiagaraRibbonRendererProperties}]
 }
 Smoke_top {
  bEnabled=false
  emitterUpdateScript [
   &EmitterState(
    Enable Distance Culling=true
    Life Cycle Mode=::NewEnumerator1
    Loop Duration=1
    Max Distance Response=::SleepAndClearParticles
    MaxDistance=1000
    Module.bUseMaxDistance=true
    Module.bUseMinDistance=false
    Scalability Mode=::NewEnumerator1
    UseLoopDelay=false
   )
   &SpawnRate(SpawnRate=5, Use Spawn Probability=false)
  ]
  particleSpawnScript [
   &InitializeParticle(
    Color Maximum=<rgba 0.25 0.25 0.25 0.6>
    Color Minimum=<rgba 0.2969 0.2969 0.2969 0.7>
    Color Mode=::NewEnumerator2
    Lifetime=&RandomRangeFloat(Minimum=0.7, Maximum=0.9)
    Position Offset=<0 0 5>
    Sprite Rotation Mode=::NewEnumerator2
    Sprite Size Mode=::NewEnumerator1
    Uniform Sprite Size Max=19
    Uniform Sprite Size Min=15
    UsePositionOffset=true
    Write Material Random=true
   ) {module=InitializeParticle001}
   &SubUVAnimation(End Frame=4, SubUV Animation Mode:ENiagara_SubUVLookupMode=::NewEnumerator2)
   &ShapeLocation(MaskSpawnGroup=false, Non Uniform Scale=<1 1 0>, Override Seed=false, Sphere Radius=1, UseOwnerScale=true)
   &AddVelocity(Rotation Coordinate Space=::World, Velocity=<0 0 20>, Velocity Speed Scale=3)
   &InheritVelocity(Inherited Velocity Amount Scale=<-0.1 -0.1 0>, Speed Limit=true, UseSpeedThreshold=false)
   &CameraOffset(Camera Offset Amount=1)
  ]
  particleUpdateScript [
   &ParticleState()
   &SolveForcesAndVelocity(Clamp Velocity=false, Limit Acceleration=false)
   &DynamicMaterialParameters(
    Index 0 Param 2=1
    Index 0 Param 3=2
    Index 0 Param 4=&FloatFromCurve(FloatCurve=DataInterface() {propertyValues {Curve {keys [<at 0 1>,<at 1 0.22>]}}})
    Module.Param0WriteEnabled=false
    Module.Param1WriteEnabled=true
    Module.Param2WriteEnabled=true
    Module.Param3WriteEnabled=true
   )
   &SpriteRotationRate(Rotation Rate=&Multiply_Float(
     A=100
     B=&FloatFromCurve(
      FloatCurve=DataInterface() {propertyValues {Curve {keys [<at 0 1>,<at 1 0>]}}}
      Scale Curve=&RandomRangeFloat(Minimum=-1, Maximum=1)
     )
    ))
   &ScaleColor(
    Scale Alpha=&Multiply_Float(
     A=&FloatFromCurve(FloatCurve=DataInterface() {propertyValues {Curve {keys [
          {interpMode=RCIM_Cubic, tangentMode=RCTM_User, time=0, value=0}
          {arriveTangent=1, interpMode=RCIM_Cubic, leaveTangent=1, time=0.1, value=0.8}
          {arriveTangent=-1.9839, interpMode=RCIM_Cubic, leaveTangent=-1.9839, tangentMode=RCTM_User, time=0.5, value=0.5}
          {tangentMode=RCTM_User, time=1, value=0}
         ]}}})
     B=0.9
    )
    Scale RGB=&VectorFromFloat(Value=&FloatFromCurve(
      FloatCurve=DataInterface() {propertyValues {Curve {keys [<at 0 0 RCIM_Cubic>, {arriveTangent=0.25, interpMode=RCIM_Cubic, leaveTangent=0.25, time=0.5, value=1.875}, <at 1 0.25 RCIM_Cubic>]}}}
      Scale Curve=10
     ))
    ScaleA=true
    ScaleRGB=true
   )
   &ScaleSpriteSize(Scale Factor=&Vector2DFromFloat(Value=&FloatFromCurve(
      FloatCurve=DataInterface() {propertyValues {Curve {keys [
          <at 0 0.8 RCIM_Cubic>
          {arriveTangent=0.5194, interpMode=RCIM_Cubic, leaveTangent=0.5194, tangentMode=RCTM_User, time=0.4843, value=1.3751}
          <at 1 2 RCIM_Cubic>
         ]}}}
      Scale Curve=2
     )))
  ]
  rendererClasses [&NiagaraSpriteRendererProperties]
  renderers [{rendererClass=&NiagaraSpriteRendererProperties}]
 }
 Steam {rendererClasses [], renderers []}
}
summary {
 systemName=NS_Proto_PoliteDinner_Steam
 emitters [
  {bEnabled=false, emitterName=Smoke_top, rendererClasses [&NiagaraSpriteRendererProperties]}
  {emitterName=Smoke_Ribbin001, rendererClasses [&NiagaraRibbonRendererProperties]}
  {emitterName=Steam, rendererClasses []}
 ]
 userVariables []
}
